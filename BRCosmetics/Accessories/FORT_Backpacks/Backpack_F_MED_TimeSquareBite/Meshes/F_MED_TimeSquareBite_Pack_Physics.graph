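# Physics asset for the TimeSquareBite backpack cosmetic, a figure with arms, hips, head, hood and necklace bones, each approximated by a capsule.
# A large body_base capsule is the hub, with free angular joints to root, hips_base and base_necklace.
# Shoulder-elbow chains, head and both hood pieces branch from body_base, while hips and a three-link necklace hang off their base bones.
# These branches use narrow swing limits, often with locked axes, to restrict secondary motion.
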PhysicsAsset /BRCosmetics/Accessories/FORT_Backpacks/Backpack_F_MED_TimeSquareBite/Meshes/F_MED_TimeSquareBite_Pack_Physics
bodies {
 base_necklace [{shapeType=Capsule, length=0.505, radius=0.505, shapeName=base_necklace_capsule, rotation=<rot 0 -90 0>}]
 body_base [{
   shapeType=Capsule
   length=48.2142
   radius=32.13
   shapeName=body_base_capsule
   center=<0.3869 42.2809 -16.5884>
   rotation=<rot 90 180 -172.329>
  }]
 dyn_elbow_l [{shapeType=Capsule, length=5, radius=2.5, shapeName=dyn_elbow_l_capsule, center=<1.3409 0.0521 0.0982>, rotation=<rot 79.6479 3.1035 -0.0002>}]
 dyn_elbow_r [{shapeType=Capsule, length=5, radius=2.5, shapeName=dyn_elbow_r_capsule, center=<-1.3043 -0.0874 0>, rotation=<rot 90 -7.4087 -16.2491>}]
 dyn_necklace_1 [{shapeType=Capsule, length=1, radius=1, shapeName=dyn_necklace_1_capsule, rotation=<rot 90 0 0>}]
 dyn_necklace_2 [{shapeType=Capsule, length=1, radius=1, shapeName=dyn_necklace_2_capsule, rotation=<rot 90 0 0>}]
 dyn_necklace_3 [{shapeType=Capsule, length=2, radius=2, shapeName=dyn_necklace_3_capsule, center=<3.1296 0 0>, rotation=<rot 90 0 0>}]
 dyn_shoulder_l [{shapeType=Capsule, length=2, radius=2.5, shapeName=dyn_shoulder_l_capsule, rotation=<rot 90 0 0>}]
 dyn_shoulder_r [{shapeType=Capsule, length=2, radius=2, shapeName=dyn_shoulder_r_capsule, rotation=<rot 90 0 0>}]
 head [{shapeType=Capsule, length=2, radius=2, shapeName=head_capsule}]
 hips [{shapeType=Capsule, length=2, radius=1, shapeName=hips_capsule, center=<1.5 0 0>, rotation=<rot 90 0 0>}]
 hips_base [{shapeType=Capsule, length=0.505, radius=0.505, shapeName=hips_base_capsule, rotation=<rot 0 -90 0>}]
 hood_lower [{shapeType=Capsule, length=2, radius=2, shapeName=hood_lower_capsule}]
 hood_upper [{shapeType=Capsule, length=2, radius=2, shapeName=hood_upper_capsule}]
 root [{shapeType=Capsule, length=10.5, radius=6.5, shapeName=root_capsule, center=<-6.1365 4.4169 -26.4176>, rotation=<rot 0 -90 0>}]
}
constraints [
 {bone1Name=hips, bone2Name=hips_base, swing1LimitDegrees=0, swing1Motion=Locked, swing2LimitDegrees=5, twistLimitDegrees=5}
 {bone1Name=dyn_shoulder_l, bone2Name=body_base, swing1LimitDegrees=20, swing2LimitDegrees=15, twistLimitDegrees=0}
 {bone1Name=dyn_elbow_l, bone2Name=dyn_shoulder_l, swing1LimitDegrees=3, swing1Motion=Locked, swing2LimitDegrees=12, twistLimitDegrees=0}
 {bone1Name=dyn_shoulder_r, bone2Name=body_base, swing1LimitDegrees=15, swing2LimitDegrees=12, twistLimitDegrees=0, twistMotion=Locked}
 {
  bone1Name=dyn_elbow_r
  bone2Name=dyn_shoulder_r
  swing1LimitDegrees=0
  swing1Motion=Locked
  swing2LimitDegrees=10
  twistLimitDegrees=0
  twistMotion=Locked
 }
 {
  bone1Name=hips_base
  bone2Name=body_base
  swing1LimitDegrees=45
  swing1Motion=Free
  swing2LimitDegrees=45
  swing2Motion=Free
  twistLimitDegrees=45
  twistMotion=Free
 }
 {bone1Name=hood_lower, bone2Name=body_base, swing1LimitDegrees=0, swing2LimitDegrees=5, twistLimitDegrees=45, twistMotion=Locked}
 {bone1Name=hood_upper, bone2Name=body_base, swing1LimitDegrees=0, swing2LimitDegrees=5, twistLimitDegrees=45, twistMotion=Locked}
 {bone1Name=head, bone2Name=body_base, swing1LimitDegrees=3, swing2LimitDegrees=15, twistLimitDegrees=45, twistMotion=Locked}
 {
  bone1Name=body_base
  bone2Name=root
  swing1LimitDegrees=45
  swing1Motion=Free
  swing2LimitDegrees=45
  swing2Motion=Free
  twistLimitDegrees=45
  twistMotion=Free
 }
 {bone1Name=dyn_necklace_3, bone2Name=dyn_necklace_2, swing1LimitDegrees=10, swing2LimitDegrees=5, twistLimitDegrees=45, twistMotion=Locked}
 {bone1Name=dyn_necklace_1, bone2Name=base_necklace, swing1LimitDegrees=20, swing2LimitDegrees=2, twistLimitDegrees=45, twistMotion=Locked}
 {bone1Name=dyn_necklace_2, bone2Name=dyn_necklace_1, swing1LimitDegrees=10, swing2LimitDegrees=5, twistLimitDegrees=45, twistMotion=Locked}
 {
  bone1Name=base_necklace
  bone2Name=body_base
  swing1LimitDegrees=45
  swing1Motion=Free
  swing2LimitDegrees=45
  swing2Motion=Free
  twistLimitDegrees=45
  twistMotion=Free
 }
]
