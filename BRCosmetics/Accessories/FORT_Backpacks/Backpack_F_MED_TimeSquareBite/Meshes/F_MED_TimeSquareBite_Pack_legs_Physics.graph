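# Physics asset for the dangling legs of the TimeSquareBite backpack cosmetic, giving its leg, head and cheek bones physical secondary motion.
# Each bone has one capsule body, with a large capsule on body_base.
# Constraints chain root, body_base, hips_base and hips, then branch into leg, knee and foot on each side with tight angular limits.
# The head and cheeks hang off body_base with locked twist, while the body_base and hips_base joints rotate freely.
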PhysicsAsset /BRCosmetics/Accessories/FORT_Backpacks/Backpack_F_MED_TimeSquareBite/Meshes/F_MED_TimeSquareBite_Pack_legs_Physics
bodies {
 body_base [{
   shapeType=Capsule
   length=48.2142
   radius=32.13
   shapeName=body_base_capsule
   center=<0.3869 42.2809 -16.5884>
   rotation=<rot 90 180 -172.329>
  }]
 cheek_l [{shapeType=Capsule, length=1, radius=1, shapeName=cheek_l_capsule}]
 cheek_r [{shapeType=Capsule, length=1, radius=1, shapeName=cheek_r_capsule}]
 dyn_foot_l [{shapeType=Capsule, length=2, radius=3.5, shapeName=dyn_foot_l_capsule, center=<1.2241 0 0>, rotation=<rot 90 0 0>}]
 dyn_foot_r [{shapeType=Capsule, length=2, radius=3.5, shapeName=dyn_foot_l_capsule, center=<1.2241 0 0>, rotation=<rot 90 180 0>}]
 dyn_knee_l [{shapeType=Capsule, length=2, radius=2, shapeName=dyn_knee_l_capsule, rotation=<rot 90 0 0>}]
 dyn_knee_r [{shapeType=Capsule, length=2, radius=2, shapeName=dyn_knee_l_capsule, rotation=<rot 90 180 0>}]
 dyn_leg_l [{shapeType=Capsule, length=7, radius=3, shapeName=dyn_leg_l_capsule, center=<4.5 0 1.5>, rotation=<rot 65.8107 -180 -180>}]
 dyn_leg_r [{shapeType=Capsule, length=4, radius=3, shapeName=dyn_leg_l_capsule, center=<1.7432 0 0>, rotation=<rot 61.5904 180 0>}]
 head [{shapeType=Capsule, length=2, radius=2, shapeName=head_capsule}]
 hips [{shapeType=Capsule, length=2, radius=1, shapeName=hips_capsule, center=<1.5 0 0>, rotation=<rot 90 0 0>}]
 hips_base [{shapeType=Capsule, length=0.505, radius=0.505, shapeName=hips_base_capsule, rotation=<rot 0 -90 0>}]
 root [{shapeType=Capsule, length=10.5, radius=6.5, shapeName=root_capsule, center=<-6.1365 4.4169 -26.4176>, rotation=<rot 0 -90 0>}]
}
constraints [
 {bone1Name=hips, bone2Name=hips_base, swing1LimitDegrees=0, swing1Motion=Locked, swing2LimitDegrees=5, twistLimitDegrees=5}
 {bone1Name=dyn_leg_l, bone2Name=hips, swing1LimitDegrees=30, swing2LimitDegrees=20, twistLimitDegrees=0}
 {bone1Name=dyn_knee_l, bone2Name=dyn_leg_l, swing1LimitDegrees=0, swing2LimitDegrees=10, twistLimitDegrees=0}
 {bone1Name=dyn_foot_l, bone2Name=dyn_knee_l, swing1LimitDegrees=10, swing2LimitDegrees=10, twistLimitDegrees=0}
 {bone1Name=dyn_knee_r, bone2Name=dyn_leg_r, swing1LimitDegrees=0, swing2LimitDegrees=10, twistLimitDegrees=0}
 {bone1Name=dyn_foot_r, bone2Name=dyn_knee_r, swing1LimitDegrees=10, swing2LimitDegrees=10, twistLimitDegrees=0}
 {bone1Name=dyn_leg_r, bone2Name=hips, swing1LimitDegrees=25, swing2LimitDegrees=20, twistLimitDegrees=0}
 {
  bone1Name=hips_base
  bone2Name=body_base
  swing1LimitDegrees=45
  swing1Motion=Free
  swing2LimitDegrees=45
  swing2Motion=Free
  twistLimitDegrees=45
  twistMotion=Free
 }
 {bone1Name=head, bone2Name=body_base, swing1LimitDegrees=3, swing2LimitDegrees=15, twistLimitDegrees=45, twistMotion=Locked}
 {bone1Name=cheek_l, bone2Name=head, swing1LimitDegrees=10, swing2LimitDegrees=10, twistLimitDegrees=45, twistMotion=Locked}
 {bone1Name=cheek_r, bone2Name=head, swing1LimitDegrees=10, swing2LimitDegrees=10, twistLimitDegrees=45, twistMotion=Locked}
 {
  bone1Name=body_base
  bone2Name=root
  swing1LimitDegrees=45
  swing1Motion=Free
  swing2LimitDegrees=45
  swing2Motion=Free
  twistLimitDegrees=45
  twistMotion=Free
 }
]
