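# Animation Blueprint for the TimeSquareBite back bling, playing idle and elimination animations with secondary motion.
# The cached state machine pose feeds RigidBody, AnimDynamics on ears, hood and head, then additive bone offsets driven by player state, weapon and spine twist.
# A kill-stat watcher bypasses that procedural chain for five seconds, while an anim input event triggers the elimination state.
# A final legs RigidBody processes either blended result.
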AnimBlueprint /BRCosmetics/Accessories/FORT_Backpacks/Backpack_F_MED_TimeSquareBite/Meshes/F_MED_TimeSquareBite_Pack_AnimBP
target_skeleton=&/F_MED_TimeSquareBite_Pack_Skeleton
graphs [
 AnimGraph {nodes [
   n1 = Use cached pose 'main loop'()
   n2 = Use cached pose 'Anims'()
   n3 = Get Active_Kill() -> (Active_Kill:bool)
   n4 = RigidBody!(
    ComponentPose=Local To Component(LocalPose=Blend Poses by bool(BlendPose_0=@n2, BlendPose_1=@n1, BlendTime_0=0.5, BlendTime_1=0.6, bActiveValue=@n3) {blendType=Linear})
    Alpha
    OverrideWorldGravity:Vector="0,0,0"
    ExternalForce
    ComponentLinearAccScale:Vector="0,0,0"
    ComponentLinearVelScale:Vector="0,0,0"
    SimSpaceSettings:Sim Space Settings Structure=(WorldAlpha=0,VelocityScaleZ=1,DampingAlpha=1,MaxLinearVelocity=10000,MaxAngularVelocity=10000,MaxLinearAcceleration=10000,MaxAngularAcceleration=10000,ExternalLinearDragV=(X=0,Y=0,Z=0),ExternalLinearVelocity=(X=0,Y=0,Z=0),ExternalAngularVelocity=(X=0,Y=0,Z=0))
   ) -> (Pose) {
    bOverrideWorldGravity=true
    overridePhysicsAsset=&/F_MED_TimeSquareBite_Pack_legs_Physics
    simulationSpace=BaseBoneSpace
    alphaBoolBlend {blendOption=Linear}
    alphaScaleBiasClamp {bInterpResult=true, interpSpeedDecreasing=5, interpSpeedIncreasing=5}
    baseBoneRef {boneName=pack}
    componentAppliedLinearAccClamp=<3000 3000 3000>
    componentLinearAccScale=<0.25 0.3 0.3>
    componentLinearVelScale=<0.25 0.3 0.3>
    simSpaceSettings {
     maxAngularAcceleration=1000
     maxAngularVelocity=3
     maxLinearAcceleration=700
     maxLinearVelocity=500
     velocityScaleZ=0
     worldAlpha=0.35
     externalLinearDragV=<0.2 0.2 0.2>
    }
   }
   Output Pose(Result=Component To Local(ComponentPose=@n4)) {layerGroup=None}
   n5 = RigidBody(
    ComponentPose=Local To Component(LocalPose=Use cached pose 'Anims'())
    OverrideWorldGravity="0,0,0"
    ComponentLinearAccScale="0,0,0"
    ComponentLinearVelScale="0,0,0"
    SimSpaceSettings=(WorldAlpha=0,VelocityScaleZ=1,DampingAlpha=1,MaxLinearVelocity=10000,MaxAngularVelocity=10000,MaxLinearAcceleration=10000,MaxAngularAcceleration=10000,ExternalLinearDragV=(X=0,Y=0,Z=0),ExternalLinearVelocity=(X=0,Y=0,Z=0),ExternalAngularVelocity=(X=0,Y=0,Z=0))
   ) {
    bOverrideWorldGravity=true
    overridePhysicsAsset=&/F_MED_TimeSquareBite_Pack_Physics
    simulationSpace=BaseBoneSpace
    alphaBoolBlend {blendOption=Linear}
    alphaScaleBiasClamp {bInterpResult=true, interpSpeedDecreasing=5, interpSpeedIncreasing=5}
    baseBoneRef {boneName=pack}
    componentAppliedLinearAccClamp=<3000 3000 3000>
    componentLinearAccScale=<0.25 0.3 0.3>
    componentLinearVelScale=<0.25 0.3 0.3>
    simSpaceSettings {
     maxAngularAcceleration=1000
     maxAngularVelocity=3
     maxLinearAcceleration=700
     maxLinearVelocity=500
     velocityScaleZ=0
     worldAlpha=0.35
     externalLinearDragV=<0.2 0.2 0.2>
    }
   }
   n6 = Get Not kill() -> (Not kill:double)
   n7 = Anim Dynamics - Bone: dyn_ear_l_02!(ComponentPose=@n5, Alpha=@n6, GravityOverride:Vector="0,0,0", ExternalForce="0,0,0") -> (Pose) {
    angularDampingOverride=0.9
    angularSpringConstant=20000
    bAngularSpring=true
    bOverrideAngularDamping=true
    bUseGravityOverride=true
    gravityScale=0.35
    simulationSpace=World
    alphaBoolBlend {blendOption=Linear}
    boundBone {boneName=dyn_ear_l_02}
    componentLinearAccScale=<1 1 1>
    componentLinearVelScale=<1 1 1>
    gravityOverride {z=-980}
    physicsBodyDefinitions [{
      boundBone {boneName=dyn_ear_l_02}
      boxExtents=<30 30 30>
      constraintSetup {angularLimitsMax=<30 7 5>, angularLimitsMin=<-30 -40 -5>, angularTarget=<1 0 0>}
      localJointOffset=<4 0 0>
     }]
    relativeSpaceBone {boneName=root}
   }
   n8 = Anim Dynamics - Bone: dyn_ear_r_02!(ComponentPose=@n7, Alpha=@n6, GravityOverride:Vector="0,0,0", ExternalForce="0,0,0") -> (Pose) {
    angularDampingOverride=1
    angularSpringConstant=20000
    bAngularSpring=true
    bOverrideAngularDamping=true
    bUseGravityOverride=true
    gravityScale=0.35
    simulationSpace=World
    alphaBoolBlend {blendOption=Linear}
    boundBone {boneName=dyn_ear_r_02}
    componentLinearAccScale=<1 1 1>
    componentLinearVelScale=<1 1 1>
    gravityOverride {z=-980}
    physicsBodyDefinitions [{
      boundBone {boneName=dyn_ear_r_02}
      boxExtents=<20 30 30>
      constraintSetup {angularLimitsMax=<15 15 15>, angularLimitsMin=<-15 -15 -15>, angularTarget=<1 0 0>}
      localJointOffset=<-4 0 0>
     }]
    relativeSpaceBone {boneName=root}
   }
   n9 = Anim Dynamics - Bone: hood_upper!(ComponentPose=@n8, Alpha=@n6, GravityOverride:Vector="0,0,0", ExternalForce="0,0,0") -> (Pose) {
    angularDampingOverride=0.99
    angularSpringConstant=10000
    bAngularSpring=true
    bOverrideAngularDamping=true
    bUseGravityOverride=true
    gravityScale=0.35
    simulationSpace=BoneRelative
    alphaBoolBlend {blendOption=Linear}
    boundBone {boneName=hood_upper}
    componentLinearAccScale=<1 1 1>
    componentLinearVelScale=<1 1 1>
    gravityOverride {z=-980}
    physicsBodyDefinitions [{
      boundBone {boneName=hood_upper}
      boxExtents=<40 30 30>
      constraintSetup {angularLimitsMax=<0 7 0>, angularLimitsMin=<0 -7 0>, angularTarget=<1 0 0>}
      localJointOffset=<1 0 0>
     }]
    relativeSpaceBone {boneName=root}
   }
   n10 = Anim Dynamics - Bone: hood_lower!(ComponentPose=@n9, Alpha=@n6, GravityOverride:Vector="0,0,0", ExternalForce="0,0,0") -> (Pose) {
    angularDampingOverride=0.999
    angularSpringConstant=15000
    bAngularSpring=true
    bOverrideAngularDamping=true
    bUseGravityOverride=true
    gravityScale=0.35
    simulationSpace=BoneRelative
    alphaBoolBlend {blendOption=Linear}
    boundBone {boneName=hood_lower}
    componentLinearAccScale=<0.5 0.5 0.5>
    componentLinearVelScale=<1 1 1>
    gravityOverride {z=-980}
    physicsBodyDefinitions [{
      boundBone {boneName=hood_lower}
      boxExtents=<20 30 30>
      constraintSetup {angularLimitsMax=<0 3 0>, angularLimitsMin=<0 0 0>, angularTarget=<1 0 0>}
      localJointOffset=<2 0 0>
     }]
    relativeSpaceBone {boneName=root}
   }
   n11 = Anim Dynamics - Bone: head!(ComponentPose=@n10, Alpha=@n6, GravityOverride:Vector="0,0,0", ExternalForce="0,0,0") -> (Pose) {
    angularDampingOverride=1
    angularSpringConstant=20000
    bAngularSpring=true
    bOverrideAngularDamping=true
    bUseGravityOverride=true
    gravityScale=0.35
    simulationSpace=BoneRelative
    alphaBoolBlend {blendOption=Linear}
    boundBone {boneName=head}
    componentLinearAccScale=<0.5 0.5 0.5>
    componentLinearVelScale=<1 1 1>
    gravityOverride {z=-980}
    physicsBodyDefinitions [{
      boundBone {boneName=head}
      boxExtents=<20 30 30>
      constraintSetup {angularLimitsMax=<15 7 5>, angularLimitsMin=<-15 -15 -5>, angularTarget=<1 0 0>}
      localJointOffset=<4 0 0>
     }]
    relativeSpaceBone {boneName=root}
   }
   n12 = Get Alpha is Floating() -> (Alpha is Floating:double)
   n13 = "Transform (Modify) Bone - Bone: dyn_tail_1"(ComponentPose=@n11, Alpha=@n12, Translation="0,0,0", Rotation="0,-20,0", Scale="1,1,1") {
    rotationMode=BMM_Additive
    rotationSpace=BCS_BoneSpace
    scaleMode=BMM_Additive
    scaleSpace=BCS_BoneSpace
    translationMode=BMM_Additive
    translationSpace=BCS_BoneSpace
    alphaBoolBlend {blendOption=Linear}
    alphaScaleBiasClamp {bInterpResult=true, interpSpeedDecreasing=3}
    boneToModify {boneName=dyn_tail_1}
    rotation {yaw=-20}
   }
   n14 = "Transform (Modify) Bone - Bone: dyn_tail_2"(ComponentPose=@n13, Alpha=@n12, Translation="0,0,0", Rotation="0,-10,0", Scale="1,1,1") {
    rotationMode=BMM_Additive
    rotationSpace=BCS_BoneSpace
    scaleMode=BMM_Additive
    scaleSpace=BCS_BoneSpace
    translationMode=BMM_Additive
    translationSpace=BCS_BoneSpace
    alphaBoolBlend {blendOption=Linear}
    alphaScaleBiasClamp {bInterpResult=true, interpSpeedDecreasing=3}
    boneToModify {boneName=dyn_tail_2}
    rotation {yaw=-10}
   }
   n15 = "Transform (Modify) Bone - Bone: dyn_tail_3"(ComponentPose=@n14, Alpha=@n12, Translation="0,0,0", Rotation="0,-15,0", Scale="1,1,1") {
    rotationMode=BMM_Additive
    rotationSpace=BCS_BoneSpace
    scaleMode=BMM_Additive
    scaleSpace=BCS_BoneSpace
    translationMode=BMM_Additive
    translationSpace=BCS_BoneSpace
    alphaBoolBlend {blendOption=Linear}
    alphaScaleBiasClamp {bInterpResult=true, interpSpeedDecreasing=3}
    boneToModify {boneName=dyn_tail_3}
    rotation {yaw=-15}
   }
   n16 = "Transform (Modify) Bone - Bone: dyn_tail_1"(
    ComponentPose=@n15
    Alpha=Get Alpha is SkyDiveMode() -> (Alpha is SkyDiveMode:double)
    Translation="0,0,0"
    Rotation="0,-10,0"
    Scale="1,1,1"
   ) {
    rotationMode=BMM_Additive
    rotationSpace=BCS_BoneSpace
    scaleMode=BMM_Additive
    scaleSpace=BCS_BoneSpace
    translationMode=BMM_Additive
    translationSpace=BCS_BoneSpace
    alphaBoolBlend {blendOption=Linear}
    alphaScaleBiasClamp {bInterpResult=true, interpSpeedDecreasing=3, interpSpeedIncreasing=15}
    boneToModify {boneName=dyn_tail_1}
    rotation {yaw=-10}
   }
   n17 = "Transform (Modify) Bone - Bone: body_base"(ComponentPose=@n16, Alpha=Get Alpha_isEmote() -> (Alpha_isEmote:double), Translation="0,-2,2", Rotation="0,0,12", Scale="1,1,1") {
    rotationMode=BMM_Additive
    rotationSpace=BCS_BoneSpace
    scaleMode=BMM_Additive
    scaleSpace=BCS_BoneSpace
    translationMode=BMM_Additive
    translationSpace=BCS_BoneSpace
    alphaBoolBlend {blendOption=Linear}
    alphaScaleBiasClamp {bInterpResult=true, interpSpeedIncreasing=7}
    boneToModify {boneName=body_base}
    rotation {roll=12}
    translation {y=-2, z=2}
   }
   n18 = "Transform (Modify) Bone - Bone: body_base"(
    ComponentPose=@n17
    Alpha=Get Alpha_isFloatingNOTFrontEnd() -> (Alpha_isFloatingNOTFrontEnd:double)
    Translation="0,-1.5,0"
    Rotation="0,0,0"
    Scale="1,1,1"
   ) {
    rotationMode=BMM_Additive
    rotationSpace=BCS_BoneSpace
    scaleMode=BMM_Additive
    scaleSpace=BCS_BoneSpace
    translationMode=BMM_Additive
    translationSpace=BCS_BoneSpace
    alphaBoolBlend {blendOption=Linear}
    alphaScaleBiasClamp {bInterpResult=true, interpSpeedDecreasing=3}
    boneToModify {boneName=body_base}
    translation {y=-1.5}
   }
   n19 = "Transform (Modify) Bone - Bone: neck"(
    ComponentPose=@n18
    Alpha=Get Alpha is Sliding() -> (Alpha is Sliding:double)
    Translation="0,0,0"
    Rotation=Get rotatorWave_Y() -> (rotatorWave_Y:Rotator)
    Scale="1,1,1"
   ) {
    rotationMode=BMM_Additive
    rotationSpace=BCS_BoneSpace
    scaleMode=BMM_Additive
    scaleSpace=BCS_BoneSpace
    translationMode=BMM_Additive
    translationSpace=BCS_BoneSpace
    alphaBoolBlend {blendOption=Linear}
    alphaScaleBiasClamp {bInterpResult=true, interpSpeedDecreasing=3}
    boneToModify {boneName=neck}
   }
   n20 = "Transform (Modify) Bone - Bone: body_base"(ComponentPose=@n19, Alpha=1, Translation=Get backpack_translation(), Rotation=Get backpack_rotate(), Scale="1,1,1") {
    rotationMode=BMM_Additive
    rotationSpace=BCS_BoneSpace
    scaleMode=BMM_Additive
    scaleSpace=BCS_BoneSpace
    translationMode=BMM_Additive
    translationSpace=BCS_BoneSpace
    alphaBoolBlend {blendOption=Linear}
    alphaScaleBiasClamp {bInterpResult=true, interpSpeedDecreasing=3}
    boneToModify {boneName=body_base}
   }
   n21 = "Transform (Modify) Bone - Bone: dyn_tail_1"(
    ComponentPose=@n20
    Alpha=Get Alpha_IsRocketLauncherInCrouching() -> (Alpha_IsRocketLauncherInCrouching:double)
    Translation="0,0,0"
    Rotation="0,-6,0"
    Scale="1,1,1"
   ) {
    rotationMode=BMM_Additive
    rotationSpace=BCS_BoneSpace
    scaleMode=BMM_Additive
    scaleSpace=BCS_BoneSpace
    translationMode=BMM_Additive
    translationSpace=BCS_BoneSpace
    alphaBoolBlend {blendOption=Linear}
    alphaScaleBiasClamp {bInterpResult=true, interpSpeedDecreasing=3}
    boneToModify {boneName=dyn_tail_1}
    rotation {yaw=-6}
   }
   n22 = "Transform (Modify) Bone - Bone: dyn_tail_1"(
    ComponentPose=@n21
    Alpha=Get Alpha_IsMinigunInCrouching() -> (Alpha_IsMinigunInCrouching:double)
    Translation="0,0,0"
    Rotation="0,-14,0"
    Scale="1,1,1"
   ) {
    rotationMode=BMM_Additive
    rotationSpace=BCS_BoneSpace
    scaleMode=BMM_Additive
    scaleSpace=BCS_BoneSpace
    translationMode=BMM_Additive
    translationSpace=BCS_BoneSpace
    alphaBoolBlend {blendOption=Linear}
    alphaScaleBiasClamp {bInterpResult=true, interpSpeedDecreasing=3}
    boneToModify {boneName=dyn_tail_1}
    rotation {yaw=-14}
   }
   n23 = "Transform (Modify) Bone - Bone: pack"(
    ComponentPose=@n22
    Alpha=Get Alpha_IsFloatingInFrontEnd() -> (Alpha_IsFloatingInFrontEnd:double)
    Translation="0,3.5,-2.5"
    Rotation="0,0,0"
   ) {
    rotationMode=BMM_Additive
    rotationSpace=BCS_BoneSpace
    scaleMode=BMM_Additive
    scaleSpace=BCS_BoneSpace
    translationMode=BMM_Additive
    translationSpace=BCS_BoneSpace
    alphaBoolBlend {blendOption=Linear}
    alphaScaleBiasClamp {bInterpResult=true, interpSpeedDecreasing=3}
    boneToModify {boneName=pack}
    translation {y=3.5, z=-2.5}
   }
   n24 = "Transform (Modify) Bone - Bone: pack"(ComponentPose=@n23, Alpha=Get Alpha_IsFrontEndIdle() -> (Alpha_IsFrontEndIdle:double), Translation="0,-1.8,0", Rotation="0,0,0") {
    rotationMode=BMM_Additive
    rotationSpace=BCS_BoneSpace
    scaleMode=BMM_Additive
    scaleSpace=BCS_BoneSpace
    translationMode=BMM_Additive
    translationSpace=BCS_BoneSpace
    alphaBoolBlend {blendOption=Linear}
    alphaScaleBiasClamp {bInterpResult=true, interpSpeedDecreasing=3}
    boneToModify {boneName=pack}
    translation {y=-1.8}
   }
   n25 = Get Alpha_IsDBNO() -> (Alpha_IsDBNO:double)
   n26 = "Transform (Modify) Bone - Bone: pack"(ComponentPose=@n24, Alpha=@n25, Translation="0,-3,0", Rotation="0,0,0") {
    rotationMode=BMM_Additive
    rotationSpace=BCS_BoneSpace
    scaleMode=BMM_Additive
    scaleSpace=BCS_BoneSpace
    translationMode=BMM_Additive
    translationSpace=BCS_BoneSpace
    alphaBoolBlend {blendOption=Linear}
    alphaScaleBiasClamp {bInterpResult=true, interpSpeedDecreasing=3}
    boneToModify {boneName=pack}
    translation {y=-3}
   }
   n27 = "Transform (Modify) Bone - Bone: dyn_tail_1"(ComponentPose=@n26, Alpha=@n25, Translation="0,0,0", Rotation="0,5,0", Scale="1,1,1") {
    rotationMode=BMM_Additive
    rotationSpace=BCS_BoneSpace
    scaleMode=BMM_Additive
    scaleSpace=BCS_BoneSpace
    translationMode=BMM_Additive
    translationSpace=BCS_BoneSpace
    alphaBoolBlend {blendOption=Linear}
    alphaScaleBiasClamp {bInterpResult=true, interpSpeedDecreasing=3}
    boneToModify {boneName=dyn_tail_1}
    rotation {yaw=5}
   }
   Save cached pose 'main loop'(Pose=Component To Local(ComponentPose=@n27)) {cacheName=main loop}
   AnimGraphNode_StateMachine_0 = Idle and Elimination() {is=title AnimGraphNode_StateMachine}
   n28 = "Transform (Modify) Bone - Bone: spine_1"(
    ComponentPose=Local To Component(LocalPose=@AnimGraphNode_StateMachine_0)
    Alpha=Get Kill_Happened() -> (Kill_Happened:double)
    Translation="0.1,0,0"
    Rotation="0,0,0"
    Scale="1,1,1"
   ) {
    translationMode=BMM_Additive
    translationSpace=BCS_BoneSpace
    alphaBoolBlend {blendOption=Linear}
    alphaScaleBiasClamp {bInterpResult=true, interpSpeedDecreasing=5, interpSpeedIncreasing=2}
    boneToModify {boneName=spine_1}
    translation {x=0.1}
   }
   n29 = "Transform (Modify) Bone - Bone: pack"(ComponentPose=@n28, Translation="0,-2.2,0", Rotation="0,0,0") {
    translationMode=BMM_Additive
    translationSpace=BCS_BoneSpace
    alphaBoolBlend {blendOption=Linear}
    boneToModify {boneName=pack}
    translation {y=-2.2}
   }
   n30 = "Transform (Modify) Bone - Bone: pack"!(ComponentPose=@n29, bAlphaBoolEnabled:bool=Get M_MED() -> (M_MED:bool), Translation="0,2,0", Rotation="0,0,0", Scale) -> (Pose) {
    alphaInputType=Bool
    translationMode=BMM_Additive
    translationSpace=BCS_BoneSpace
    alphaBoolBlend {blendOption=Linear}
    boneToModify {boneName=pack}
    translation {y=2}
   }
   n31 = "Transform (Modify) Bone - Bone: pack"(ComponentPose=@n30, bAlphaBoolEnabled=Get F_MED() -> (F_MED:bool), Translation="0,3.5,0", Rotation="0,0,0") {
    alphaInputType=Bool
    translationMode=BMM_Additive
    translationSpace=BCS_BoneSpace
    alphaBoolBlend {blendOption=Linear}
    boneToModify {boneName=pack}
    translation {y=3.5}
   }
   Save cached pose 'Anims'(Pose=Component To Local(ComponentPose=@n31)) {cacheName=Anims}
  ]}
 Idle and Elimination {
  entry=Idle
  path=AnimGraph.AnimGraphNode_StateMachine_0.Idle and Elimination
  ids=id
  nodes [
   AnimStateNode_0 = Idle() {
    stateEntered {guid=#1}
    stateFullyBlended {guid=#2}
    stateLeft {guid=#3}
   }
   AnimStateNode_1 = Elimination() {
    is=title AnimStateNode
    stateEntered {guid=#4}
    stateFullyBlended {guid=#5}
    stateLeft {guid=#6}
   }
   AnimStateTransitionNode_0 = AnimStateTransitionNode() {
    transitionEnd {guid=#7}
    transitionInterrupt {guid=#8}
    transitionStart {guid=#9}
   }
   AnimStateTransitionNode_1 = AnimStateTransitionNode() {bAutomaticRuleBasedOnSequencePlayerInState=true}
  ]
  links [
   {
    dst=Idle
    rule=AnimGraph.AnimGraphNode_StateMachine_0.Idle and Elimination.AnimStateTransitionNode_1.Transition
    src=Elimination
    via=AnimStateTransitionNode_1
   }
   {
    dst=Elimination
    rule=AnimGraph.AnimGraphNode_StateMachine_0.Idle and Elimination.AnimStateTransitionNode_0.Transition
    src=Idle
    via=AnimStateTransitionNode_0
   }
  ]
 }
 Idle {
  path=AnimGraph.AnimGraphNode_StateMachine_0.Idle and Elimination.AnimStateNode_0.Idle
  nodes [Output Animation Pose(Result=Sequence Player 'TimeSquareBite_Idle_PACK'() {sequence=/BRCosmetics/Accessories/FORT_Backpacks/Backpack_F_MED_TimeSquareBite/Animations/TimeSquareBite_Idle_PACK})]
 }
 Elimination {
  path=AnimGraph.AnimGraphNode_StateMachine_0.Idle and Elimination.AnimStateNode_1.Elimination
  nodes [
   n1 = Sequence Player 'TimeSquareBite_Elim_PACK'() {bLoopAnimation=false, sequence=/BRCosmetics/Accessories/FORT_Backpacks/Backpack_F_MED_TimeSquareBite/Animations/TimeSquareBite_Elim_PACK}
   Output Animation Pose(Result=@n1)
  ]
 }
 Transition {
  path=AnimGraph.AnimGraphNode_StateMachine_0.Idle and Elimination.AnimStateTransitionNode_0.Transition
  nodes [Result(bCanEnterTransition=Get bEnemyPlayerEliminated() -> (bEnemyPlayerEliminated:bool))]
 }
 Transition {path=AnimGraph.AnimGraphNode_StateMachine_0.Idle and Elimination.AnimStateTransitionNode_1.Transition, nodes [Result()]}
 EventGraph {nodes [
   n1 = Event BlueprintProcessAnimInputEvent()
   > Branch(Condition=IsNamed(self=@n1.AnimInputEvent, InEventName=aie.EnemyPlayerEliminated))
   > Set bEnemyPlayerEliminated(bEnemyPlayerEliminated:bool=true) -> (Output_Get:bool)
   Event BlueprintPostEvaluateAnimation() {is=title K2Node_Event}
   > Set bEnemyPlayerEliminated(bEnemyPlayerEliminated=false)
   Event BlueprintUpdateAnimation()
   > n2 = Sequence!(execute) -> (then_0, then_1, then_2:Exec, then_3:Exec, then_4:Exec)
   n3 = Branch(execute=@n2.then_2, Condition=Get bIsSliding())
   > Set speedWave(speedWave:double=10) -> (Output_Get:double)
   > Set amplitudWave(amplitudWave:double=1.5) -> (Output_Get:double)
   > Set offsetWave(offsetWave:double=0) -> (Output_Get:double)
   Set speedWave(execute=@n3.else, speedWave=0)
   n4 = Sequence!(execute=@n2.then_3) -> (
    then_0
    then_1
    then_2
    then_3
    then_4
    then_5:Exec
    then_6:Exec
    then_7:Exec
    then_8:Exec
    then_9:Exec
    then_10:Exec
    then_11:Exec
    then_12:Exec
   )
   Get!(execute:Exec=@n4.then_12) -> (then:Exec, else:Exec, Skel_Mesh:Skeletal Mesh Component Object Reference)
   > Sequence!(execute) -> (then_0, then_1)
   > n5 = Branch(Condition=Get Active_Kill())
   > "Set FInterp(Alpha, (10: 10))"("FInterp(Alpha, (10: 10))":float="Get FInterp(Alpha, (10: 10))_0"() -> ("FInterp(Alpha, (10: 10))_0":float)) -> (Output_Get:float)
   > Delay(Duration=5)
   > Set Active_Kill(Active_Kill:bool=false) -> (Output_Get:bool)
   "Set FInterp(Alpha, (10: 10))"(execute=@n5.else, "FInterp(Alpha, (10: 10))"=0)
   n6 = Branch(execute=@n4.then_4, Condition=Get bIsInFrontEnd())
   > n7 = Branch(Condition=Get bIsSkyDiving())
   > Set Component Linear Acc Scale(Component Linear Acc Scale:Vector="0.3,0.3,0.3") -> (Output_Get:Vector)
   > Set Component Linear Vel Scale(Component Linear Vel Scale:Vector="0.3,0.3,0.3") -> (Output_Get:Vector)
   > Set Alpha is RigBodyAlphaModify(Alpha is RigBodyAlphaModify:double=0.5) -> (Output_Get:double)
   n8 = Branch(execute=@n7.else, Condition=Get bIsPlayingEmote())
   > Set Component Linear Acc Scale(Component Linear Acc Scale="0.4,0.4,0.4")
   > Set Component Linear Vel Scale(Component Linear Vel Scale="0.4,0.4,0.4")
   > Set Alpha is RigBodyAlphaModify(Alpha is RigBodyAlphaModify=1)
   Set Component Linear Acc Scale(execute=@n8.else, Component Linear Acc Scale="0.4,0.4,0.4")
   > Set Component Linear Vel Scale(Component Linear Vel Scale="0.4,0.4,0.4")
   > Set Alpha is RigBodyAlphaModify(Alpha is RigBodyAlphaModify=0.5)
   n9 = Branch(execute=@n6.else, Condition=Get bIsSkyDiving())
   > Set Component Linear Acc Scale(Component Linear Acc Scale="0.2,0.2,0.2")
   > Set Component Linear Vel Scale(Component Linear Vel Scale="0.2,0.2,0.2")
   > Set Alpha is RigBodyAlphaModify(Alpha is RigBodyAlphaModify=0.5)
   Set Component Linear Acc Scale(execute=@n9.else, Component Linear Acc Scale="0.3,0.33,0.3")
   > Set Component Linear Vel Scale(Component Linear Vel Scale="0.3,0.33,0.3")
   > Set Alpha is RigBodyAlphaModify(Alpha is RigBodyAlphaModify=1)
   n10 = Branch(execute=@n4.then_11, Condition=Get bIsDBNO())
   > Set Alpha_IsDBNO(Alpha_IsDBNO:double=1) -> (Output_Get:double)
   Set Alpha_IsDBNO(execute=@n10.else, Alpha_IsDBNO=0)
   n11 = Branch(execute=@n4.then_1, Condition=Get bIsSkydiveFloating())
   > Set Alpha is Floating(Alpha is Floating:double=1) -> (Output_Get:double)
   Set Alpha is Floating(execute=@n11.else, Alpha is Floating=0)
   n12 = Sequence!(execute=@n4.then_2) -> (then_0, then_1, then_2:Exec, then_3:Exec)
   n13 = Cast To FortPlayerAnimInstance(execute=@n12.then_1, Object=GetAnimInstance(self=GetOwnerSkeletalMeshComponent()))
   > Set ar_firingPosition(ar_firingPosition:int=1) -> (Output_Get:int=0)
   n14 = Get WeaponCoreAnim(self=@n13)
   > n15 = Switch on EFortWeaponCoreAnimation!(execute, Selection:EFortWeaponCoreAnimation Enum=@n14) -> (
    Melee:Exec
    Pistol:Exec
    Shotgun:Exec
    PaperBlueprint:Exec
    Rifle:Exec
    MeleeOneHand:Exec
    MachinePistol:Exec
    RocketLauncher:Exec
    GrenadeLauncher:Exec
    GoingCommando:Exec
    AssaultRifle:Exec
    TacticalShotgun:Exec
    SniperRifle:Exec
    TrapPlacement:Exec
    ShoulderLauncher:Exec
    AbilityDecoTool:Exec
    Crossbow:Exec
    C4:Exec
    RemoteControl:Exec
    DualWield:Exec
    AR_BullPup:Exec
    AR_ForwardGrip:Exec
    MedPackPaddles:Exec
    SMG_P90:Exec
    AR_DrumGun:Exec
    Consumable_Small:Exec
    Consumable_Large:Exec
    Balloon:Exec
    MountedTurret:Exec
    CreativeTool:Exec
    ExplosiveBow:Exec
    AshtonIndigo:Exec
    AshtonChicago:Exec
    MeleeDualWield:Exec
    Unarmed:Exec
   )
   Set Alpha_IsMinugunSelected(
    execute=[
     @n15.AR_BullPup
     @n15.AR_DrumGun
     @n15.AR_ForwardGrip
     @n15.AbilityDecoTool
     @n15.AshtonChicago
     @n15.AshtonIndigo
     @n15.AssaultRifle
     @n15.Balloon
     @n15.C4
     @n15.Consumable_Large
     @n15.Consumable_Small
     @n15.CreativeTool
     @n15.Crossbow
     @n15.DualWield
     @n15.ExplosiveBow
     @n15.GrenadeLauncher
     @n15.MachinePistol
     @n15.MedPackPaddles
     @n15.Melee
     @n15.MeleeDualWield
     @n15.MeleeOneHand
     @n15.MountedTurret
     @n15.PaperBlueprint
     @n15.Pistol
     @n15.RemoteControl
     @n15.Rifle
     @n15.RocketLauncher
     @n15.SMG_P90
     @n15.Shotgun
     @n15.ShoulderLauncher
     @n15.SniperRifle
     @n15.TacticalShotgun
     @n15.TrapPlacement
     @n15.Unarmed
    ]
    Alpha_IsMinugunSelected:bool=false
   ) -> (Output_Get:bool)
   Set Alpha_IsMinugunSelected(execute=@n15.GoingCommando, Alpha_IsMinugunSelected=true)
   n16 = Switch on EFortWeaponCoreAnimation(execute=@n12.then_0, Selection=@n14)
   Set Alpha_IsRocketLauncherSelected(
    execute=[
     @n16.AR_BullPup
     @n16.AR_DrumGun
     @n16.AR_ForwardGrip
     @n16.AbilityDecoTool
     @n16.AshtonChicago
     @n16.AshtonIndigo
     @n16.AssaultRifle
     @n16.Balloon
     @n16.C4
     @n16.Consumable_Large
     @n16.Consumable_Small
     @n16.CreativeTool
     @n16.Crossbow
     @n16.DualWield
     @n16.ExplosiveBow
     @n16.GoingCommando
     @n16.GrenadeLauncher
     @n16.MachinePistol
     @n16.MedPackPaddles
     @n16.Melee
     @n16.MeleeOneHand
     @n16.MountedTurret
     @n16.PaperBlueprint
     @n16.Pistol
     @n16.RemoteControl
     @n16.Rifle
     @n16.SMG_P90
     @n16.Shotgun
     @n16.ShoulderLauncher
     @n16.SniperRifle
     @n16.TacticalShotgun
     @n16.TrapPlacement
     @n16.Unarmed
    ]
    Alpha_IsRocketLauncherSelected:bool=false
   ) -> (Output_Get:bool)
   Set Alpha_IsRocketLauncherSelected(execute=@n16.RocketLauncher, Alpha_IsRocketLauncherSelected=true)
   n17 = Branch(execute=@n12.then_3, Condition=AND Boolean(A=Get Alpha_IsMinugunSelected() -> (Alpha_IsMinugunSelected:bool), B=Get bIsCrouching()))
   > Set Alpha_IsMinigunInCrouching(Alpha_IsMinigunInCrouching:double=1) -> (Output_Get:double)
   Set Alpha_IsMinigunInCrouching(execute=@n17.else, Alpha_IsMinigunInCrouching=0)
   n18 = Branch(
    execute=@n12.then_2
    Condition=AND Boolean(A=Get Alpha_IsRocketLauncherSelected() -> (Alpha_IsRocketLauncherSelected:bool), B=Get bIsCrouching())
   )
   > Set Alpha_IsRocketLauncherInCrouching(Alpha_IsRocketLauncherInCrouching:double=1) -> (Output_Get:double)
   Set Alpha_IsRocketLauncherInCrouching(execute=@n18.else, Alpha_IsRocketLauncherInCrouching=0)
   n19 = Branch(execute=@n4.then_3, Condition=Get bIsSkydiveDiveMode())
   > Set Alpha is SkyDiveMode(Alpha is SkyDiveMode:double=1) -> (Output_Get:double)
   Set Alpha is SkyDiveMode(execute=@n19.else, Alpha is SkyDiveMode=0)
   n20 = GetOwnerSkeletalMeshComponent()
   n21 = BreakRotator(InRot=BreakTransform(InTransform=GetDeltaTransformFromRefPose(self=@n20, BoneName=spine_05, BaseName=spine_01)).Rotation)
   Set backpack_rotate(execute=@n4.then_8, backpack_rotate=MakeRotator(Roll=MapRangeClamped(Value=@n21.Yaw, InRangeA=0, InRangeB=-50, OutRangeB=30)))
   > Set backpack_translation(backpack_translation=MakeVector(Y=MapRangeClamped(Value=@n21.Yaw, InRangeA=0, InRangeB=-30, OutRangeA=1, OutRangeB=-5)))
   n22 = Branch(execute=@n4.then_9, Condition=AND Boolean(A=Get bIsInFrontEnd(), B=Get bIsSkydiveFloating()))
   > Set Alpha_IsFloatingInFrontEnd(Alpha_IsFloatingInFrontEnd:double=1) -> (Output_Get:double)
   Set Alpha_IsFloatingInFrontEnd(execute=@n22.else, Alpha_IsFloatingInFrontEnd=0)
   n23 = Get bIsSkyDiving()
   n24 = Get bIsPlayingEmote()
   n25 = Branch(execute=@n4.then_10, Condition=AND Boolean!(A=Get bIsInFrontEnd(), B=NOT Boolean(A=@n23), C:bool=NOT Boolean(A=@n24)) -> (ReturnValue))
   > Set Alpha_IsFrontEndIdle(Alpha_IsFrontEndIdle:double=1) -> (Output_Get:double)
   Set Alpha_IsFrontEndIdle(execute=@n25.else, Alpha_IsFrontEndIdle=0)
   n26 = Branch(execute=@n4.then_0, Condition=Get bIsSliding())
   > Set Alpha is Sliding(Alpha is Sliding:double=1) -> (Output_Get:double)
   Set Alpha is Sliding(execute=@n26.else, Alpha is Sliding=0)
   n27 = Branch(execute=@n4.then_5, Condition=Get bIsPlayingEmote())
   > Set Alpha_isEmote(Alpha_isEmote:double=1) -> (Output_Get:double)
   Set Alpha_isEmote(execute=@n27.else, Alpha_isEmote=0)
   n28 = Get bIsInFrontEnd()
   n29 = Get bIsSkydiveFloating()
   n30 = Get bIsSkydiveDiveMode()
   n31 = Branch(execute=@n4.then_6, Condition=AND Boolean!(A=OR Boolean(A=@n29, B=@n30), B=NOT Boolean(A=@n28)) -> (ReturnValue))
   > Set Alpha_isFloatingNOTFrontEnd(Alpha_isFloatingNOTFrontEnd:double=1) -> (Output_Get:double)
   Set Alpha_isFloatingNOTFrontEnd(execute=@n31.else, Alpha_isFloatingNOTFrontEnd=0)
   n32 = Branch(
    execute=@n4.then_7
    Condition=OR Boolean!(A=Get bIsJumping(), B=Get bIsFalling(), C:bool=Get bIsSkyDiving(), D:bool=Get bIsParachuteOpened()) -> (ReturnValue)
   )
   > Set Sim Space Settings On air(Sim Space Settings On air:Sim Space Settings Structure=Get air() -> (air:Sim Space Settings Structure)) -> (Output_Get:Sim Space Settings Structure)
   Set Sim Space Settings On air(execute=@n32.else, Sim Space Settings On air=Get ground() -> (ground:Sim Space Settings Structure))
   n33 = Gravity Override State Function DEPRECATED(
    execute=@n2.then_0
    CharacterPartAnimInstance=Self-Reference()
    GameTimeInSeconds=GetGameTimeInSeconds()
    GravityOverrideParams=Get Gravity Override Params()
   )
   > Set GravityOverride(GravityOverride=@n33)
   n34 = Get Active_Kill()
   > Set Kill_Happened(Kill_Happened:double="To Float (Boolean)"(InBool=@n34)) -> (Output_Get:double)
   > Set Not kill(Not kill:double="To Float (Boolean)"(InBool=NOT Boolean(A=@n34))) -> (Output_Get:double)
   n35 = "Sin (Radians)"(A=float * float(A=GetRealTimeSeconds(), B=Get speedWave() -> (speedWave:double)) {is=title K2Node_PromotableOperator})
   n36 = float + float(
    A=float * float(A=MultiplyByPi!(Value:double=@n35) -> (ReturnValue:double), B=Get amplitudWave() -> (amplitudWave:double))
    B=Get offsetWave() -> (offsetWave:double)
   )
   Set rotatorWave_Z(execute=@n2.then_1, rotatorWave_Z:Rotator=MakeRotator(Yaw=@n36)) -> (Output_Get:Rotator)
   > Set rotatorWave_Y(rotatorWave_Y:Rotator=MakeRotator(Pitch=@n36)) -> (Output_Get:Rotator)
   Event OnCharacterCustomizationLoaded()
   > n37 = Sequence!(execute) -> (then_0, then_1)
   > n38 = Cast To FortPlayerPawn(Object=TryGetFortPawn())
   > n39 = WatchCosmeticStat!(
    execute:Exec
    Pawn:Fort Player Pawn Object Reference=@n38
    StatTag:Fort Stat Manager Tag Structure=(TagName="StatManager.AthenaKills")
    DefaultValue:int=0
    bBroadcastInitialValue:bool=false
    InitialValueBroadcastDelay:float=-1
    ChangeValueBroadcastDelay:float=-1
    bConsumeWaitingNotifications:bool=true
   ) -> (then:Exec, OnStatChanged:Exec, StatName:Name, StatValue:int) {is=title K2Node_AsyncAction}
   Set KillCount(execute=@n39.OnStatChanged, KillCount:double="To Float (Integer)"(InInt=@n39.StatValue)) -> (Output_Get:double)
   > Set Active_Kill(Active_Kill=true)
   > Delay(Duration=5)
   > Set Active_Kill(Active_Kill=false)
   n40 = Set Owning Pawn(execute=@n37.then_1, Owning Pawn:Pawn Object Reference=TryGetFortPawn()) -> (Output_Get:Pawn Object Reference)
   > Is Valid(InputObject=@n40)
   > n41 = Sequence()
   > n42 = FindGender!(execute, Object:Object Reference=Get Owning Pawn() -> (Owning Pawn:Pawn Object Reference)) -> (True:Exec, False:Exec, OutGender:EFortCustomGender Enum=Invalid)
   > Set Male(Male:bool="Equal (Enum)"(A:EFortCustomGender Enum=@n42, B:EFortCustomGender Enum=Male)) -> (Output_Get:bool)
   > n43 = Set Female(Female:bool="Equal (Enum)"(A=@n42, B=Female)) -> (Output_Get:bool)
   n44 = Get bIsInFrontEnd()
   Set F_MED(
    execute=[@n43.then, @n46.then]
    F_MED:bool=AND Boolean!(A=Get Female() -> (Female:bool), B=Get Medium() -> (Medium:bool), C:bool=@n44) -> (ReturnValue)
   ) -> (Output_Get:bool)
   > Set M_MED(M_MED:bool=AND Boolean(A=Get Male() -> (Male:bool), B=Get Medium(), C=@n44)) -> (Output_Get:bool)
   n45 = FindBodyType!(execute=@n41.then_1, Object:Object Reference=Get Owning Pawn()) -> (True:Exec, False:Exec, OutBodyType:EFortCustomBodyType Enum=NONE)
   > Set Large(Large:bool="Equal (Enum)"(A:EFortCustomBodyType Enum=@n45, B:EFortCustomBodyType Enum=Large)) -> (Output_Get:bool)
   > n46 = Set Medium(Medium:bool="Equal (Enum)"(A=@n45, B=Medium)) -> (Output_Get:bool)
  ]}
]
variables [
 Active_Kill {type=boolean}
 air {
  type=SimSpaceSettings
  default {
   dampingAlpha=1
   maxAngularAcceleration=1000
   maxAngularVelocity=3
   maxLinearAcceleration=700
   maxLinearVelocity=500
   velocityScaleZ=0
   worldAlpha=0.15
   externalLinearDragV=<0.2 0.2 0.2>
  }
 }
 Alpha is Floating {type=number}
 Alpha is RigBodyAlphaModify {type=number}
 Alpha is SkyDiveMode {type=number}
 Alpha is Sliding {type=number}
 Alpha_IsDBNO {type=number}
 Alpha_isEmote {type=number}
 Alpha_IsFloatingInFrontEnd {type=number}
 Alpha_isFloatingNOTFrontEnd {type=number}
 Alpha_IsFrontEndIdle {type=number}
 Alpha_IsMinigunInCrouching {type=number}
 Alpha_IsMinugunSelected {type=boolean}
 Alpha_isMoving2DCrouching {type=number}
 Alpha_IsRocketLauncherInCrouching {type=number}
 Alpha_IsRocketLauncherSelected {type=boolean}
 amplitudWave {type=number}
 ar_firingPosition {type=integer}
 backpack_rotate {type=Rotator}
 backpack_translation {type=Vector}
 bEnemyPlayerEliminated {type=boolean}
 Component Linear Acc Scale {type=Vector}
 Component Linear Vel Scale {type=Vector}
 F_MED {type=boolean}
 Female {type=boolean}
 "FInterp(Alpha, (10: 10))" {type=number}
 "FInterp(Alpha, (10: 10))_0" {type=number}
 Gravity Override Params {
  type=GravityOverrideParamsStruct
  default {
   skydive_FloatingDown {gravityOverride=<0 -700 -600>, windAmplitude=<200 -600 400>, windOffset=<0 -500 400>}
   skydive_Parachuting {gravityOverride=<0 0 -980>, windAmplitude=<0 0 0>, windFrequency=<0 0 0>}
   tactical_Sprint {gravityOverride=<0 0 -500>}
  }
 }
 GravityOverride {type=Vector}
 ground {
  type=SimSpaceSettings
  default {
   dampingAlpha=1
   maxAngularAcceleration=1000
   maxAngularVelocity=3
   maxLinearAcceleration=700
   maxLinearVelocity=500
   velocityScaleZ=0
   worldAlpha=0.5
   externalLinearDragV=<0.2 0.2 0.2>
  }
 }
 Kill_Happened {type=number}
 KillCount {type=number}
 Large {type=boolean}
 M_MED {type=boolean}
 Male {type=boolean}
 Medium {type=boolean}
 Not kill {type=number}
 offsetWave {type=number}
 Owning Pawn {type=Pawn, default=null}
 rotatorWave_Y {type=Rotator}
 rotatorWave_Z {type=Rotator}
 Sim Space Settings On air {
  type=SimSpaceSettings
  default {
   dampingAlpha=1
   maxAngularAcceleration=1000
   maxAngularVelocity=3
   maxLinearAcceleration=700
   maxLinearVelocity=500
   velocityScaleZ=0
   worldAlpha=0.1
   externalLinearDragV=<0.2 0.2 0.2>
  }
 }
 Skel_Mesh {type=SkeletalMeshComponent}
 speedWave {type=number}
]
